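# Material function that adds a two-color emissive glow, also tinted into base color, to Demon Island materials.
# Inverted detail and mask inputs shape the glow, which blends over the incoming material attributes.
# Local-space Z gradients, one eroded by inverted noise, provide lower-region falloff masks via named reroutes.
# MaterialProxyReplace substitutes a world-normal mask for height blending in proxy bakes; a Fresnel branch is unconnected.
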MaterialFunction /DemonArenaGameplay/Environment/Materials/MF_DemonIslandGlow
graph {nodes [
  n1 = NamedRerouteUsage()
  n2 = LinearInterpolate(
   A=VectorParameter() {parameterName=DemonIsland_GlowColor, defaultValue=<rgba 0.0932 0 0.2431 1>}.RGB
   B=VectorParameter() {parameterName=DemonIsland_GlowColor2, defaultValue=<rgba 0.2431 0.0405 0 1>}.RGB
   Alpha=NamedRerouteUsage()
  )
  n3 = NamedRerouteDeclaration(None=FunctionInput() {inputName=DetailTexture, inputType=FunctionInput_Scalar}) {name=ConvexMask, nodeColor=<rgba 0 1 0.4375 1>}
  n4 = Multiply(A=Power(Base=OneMinus(None=@n3)) {constExponent=3}) {constB=2}
  n5 = FunctionInput(Preview=Constant()) {bUsePreviewValueAsDefault=true, inputName=Mask, inputType=FunctionInput_Scalar}
  n6 = Saturate(None=Max(A=@n4, B=OneMinus(None=@n5)))
  n7 = Multiply(A=Multiply(A=Multiply(A=@n2, B=@n4), B=@n6), B=ScalarParameter() {defaultValue=5, parameterName=DemonIsland_Glow_Strength})
  n8 = LinearInterpolate(A=NamedRerouteUsage(), B=@n7, Alpha=@n6)
  FunctionOutput(None=BlendMaterialAttributes(
    A=@n1
    B=SetMaterialAttributes!(MaterialAttributes=@n1, EmissiveColor=@n7, BaseColor=@n8) {attributeSetTypes [#1, #2]}
    Alpha=NamedRerouteUsage()
   ))
  n9 = GetMaterialAttributes(None=FunctionInput(Preview=SetMaterialAttributes!(MaterialAttributes)) {bUsePreviewValueAsDefault=true, inputType=FunctionInput_MaterialAttributes}) {attributeGetTypes [#3, #2]}
  NamedRerouteDeclaration(None=@n9.MaterialAttributes) {name=Input_Material, nodeColor=<rgba 0.9219 1 0 1>}
  n10 = LocalPosition() {localOrigin=InstancePreSkinning}
  n11 = ShaderStageSwitch(PixelShader=VertexInterpolator(VS=@n10.XYZ).PS, VertexShader=@n10.XYZ)
  n12 = Subtract(A=@n11, B=VectorParameter() {parameterName=DemonIslandBlendGradient, defaultValue=<rgba 0 0 -3000 1>}.RGB)
  n13 = ScalarParameter() {defaultValue=2500, parameterName=DemonIslandBlendFalloff}
  n14 = Saturate(None=ComponentMask(None=Divide(A=@n12, B=@n13)) {b=true})
  NamedRerouteDeclaration(None=MaterialProxyReplace(Realtime=OneMinus(None=@n14), MaterialProxy=NamedRerouteUsage())) {name=Mask, nodeColor=<rgba 0 0.7656 1 1>}
  n15 = Divide(
   A=Subtract(A=@n11, B=VectorParameter() {parameterName=DemonIslandGlowOffset, defaultValue=<rgba 0 0 -3000 1>}.RGB)
   B=ScalarParameter() {defaultValue=2000, parameterName=DemonIslandGradientFalloff}
  )
  n16 = OneMinus(None=Saturate(None=ComponentMask(None=@n15) {b=true}))
  n17 = Multiply(A=NamedRerouteUsage(), B=ScalarParameter() {defaultValue=0.2, parameterName=EmissiveNoiseStrength})
  NamedRerouteDeclaration(None=Saturate(None=Subtract(A=@n16, B=@n17))) {name=FalloffGradient, nodeColor=<rgba 1 0.1875 0 1>}
  NamedRerouteDeclaration(None=@n9.BaseColor) {name=Input_BaseColor, nodeColor=<rgba 0.8984 0 1 1>}
  n18 = Constant()
  n19 = Power(Base=OneMinus(None=FunctionInput(Preview=@n18) {bUsePreviewValueAsDefault=true, inputName=NoiseTexture, inputType=FunctionInput_Scalar})) {constExponent=3}
  NamedRerouteDeclaration(None=Saturate(None=Multiply(A=@n19) {constB=2})) {name=Noise, nodeColor=<rgba 0 0.1562 1 1>}
  n20 = Multiply(A=PixelNormalWS(), B=Constant3Vector() {constant=<rgba 0 0 1 1>})
  n21 = OneMinus(None=ComponentMask(None=Divide(A=@n20) {constB=0.015}) {b=true})
  NamedRerouteDeclaration(None=Saturate(None=Divide(A=@n21) {constB=65})) {name=ProxyMaskReplacement, nodeColor=<rgba 1 0 0.6797 1>}
  n22 = Transform(None=@n9.Normal)
  Saturate(None=OneMinus(None=Fresnel(Normal=@n22) {exponent=1}))
 ]}
