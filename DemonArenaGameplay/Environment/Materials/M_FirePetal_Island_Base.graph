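# Layered Nanite master material, apparently for Demon Arena island rock meshes, combining baked textures with landscape-matched terrain layers.
# Grass, gravel or custom world-projected textures supply terrain layers, with world-aligned slope blending and noise-broken vertex-color masks available as blend inputs.
# A demon island glow function follows the terrain blend.
# Static, quality and shading-path switches limit detail normals to Epic, and the optional height-blended second layer to non-baked, non-mobile, Medium-and-up rendering.
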Material /DemonArenaGameplay/Environment/Materials/M_FirePetal_Island_Base
graph {
 properties {bUseMaterialAttributes=true, bUsedWithNanite=true, pixelDepthOffsetMode=PDOM_Legacy, usedShadingModels=Default Lit}
 nodes [
  MaterialFunctionCall() {materialFunction=/Game/Environments/Apollo/Sets/Abducted/Chunks/Materials/MaterialFunctions/MF_Abducted_Chunk_Toplayer}
  n1 = VectorParameter() {group=Additional Layer, parameterName=Additional Layer Modulate, defaultValue=<rgba 1 1 1 1>}
  n2 = Multiply(A=NamedRerouteUsage(), B=@n1.RGB)
  MaterialFunctionCall!(Additional Layer Modulate=ShadingPathSwitch(
    Default=StaticSwitchParameter(True=@n2, False=@n1.RGB) {defaultValue=true, group=Additional Layer, parameterName=Large Scale Break up}
    Mobile=@n1.RGB
   )) {materialFunction=/Game/Environments/Apollo/Sets/Abducted/Chunks/Materials/MaterialFunctions/MF_Abducted_Chunk_AdditionalLayer}
  n3 = MaterialFunctionCall!(Grass_UseCustomSlopeMask, Grass_CustomSlopeMask) {materialFunction=/BRCoreMaterials/Hermes/Terrain/Landscape/Materials/Layers/MF_Hermes_StaticMeshLandscapeLayerMatch}
  n4 = GetMaterialAttributes(None=@n3.GrassLayer) {attributeGetTypes [#1]}
  n5 = SetMaterialAttributes!(
   MaterialAttributes=@n4.MaterialAttributes
   Normal=Transform(None=@n4.Normal) {transformSourceType=TRANSFORMSOURCE_World, transformType=TRANSFORM_Tangent}
  ) {attributeSetTypes [#1]}
  n6 = GetMaterialAttributes(None=@n3.GravelLayer) {attributeGetTypes [#1]}
  n7 = SetMaterialAttributes(
   MaterialAttributes=@n6.MaterialAttributes
   Normal=Transform(None=@n6.Normal) {transformSourceType=TRANSFORMSOURCE_World, transformType=TRANSFORM_Tangent}
  ) {attributeSetTypes [#1]}
  NamedRerouteDeclaration(None=StaticSwitchParameter(True=@n5, False=@n7) {parameterName=UseGravelAsLayer1}) {name=TerrainLayer2, nodeColor=<rgba 1 0 0.7734 1>}
  n8 = MaterialFunctionCall!(Texture Scale=ScalarParameter() {defaultValue=1021.8221, group=Middle Layer, parameterName=Middle Layer Texture Scale}) {materialFunction=/Game/Environments/Apollo/Sets/Abducted/Chunks/Materials/MaterialFunctions/MF_Abducted_Chunk_MiddleLayer}
  NamedRerouteDeclaration(None=@n8.Result) {name=MiddleLayer, nodeColor=<rgba 1 0 0.7734 1>}
  NamedRerouteDeclaration(None=@n8.Heightmap) {name=MiddleLayerHeight, nodeColor=<rgba 1 0 0.7734 1>}
  n9 = MaterialFunctionCall!(Scale=ScalarParameter() {defaultValue=1024, group=CustomLayer, parameterName=CustomLayer_WorldScale, sortPriority=-500}) {materialFunction=/Game/Athena/Apollo/Environments/Landscape/Materials/Biome/CommonMaterials/MF_SimpleWorldUV}
  n10 = TextureSampleParameter2D(UVs=@n9.Result) {
   group=CustomLayer
   parameterName=CustomLayer_Diffuse
   samplerSource=SSM_Wrap_WorldGroupSettings
   sortPriority=-400
   texture=/Game/Environments/Helios/Terrain/Textures/T_Helios_Brim_RockGround_01_D
  }
  n11 = TextureSampleParameter2D(UVs=@n9.Result) {
   group=CustomLayer
   parameterName=CustomLayer_Spec
   samplerSource=SSM_Wrap_WorldGroupSettings
   samplerType=SAMPLERTYPE_Masks
   sortPriority=-300
   texture=/Game/Environments/Helios/Terrain/Textures/T_Helios_Brim_RockGround_01_S
  }
  n12 = TextureSampleParameter2D(UVs=@n9.Result) {
   group=CustomLayer
   parameterName=CustomLayer_Normal
   samplerSource=SSM_Wrap_WorldGroupSettings
   samplerType=SAMPLERTYPE_Normal
   sortPriority=-200
   texture=/Game/Environments/Helios/Terrain/Textures/T_Helios_Brim_RockGround_01_N
  }
  n13 = SetMaterialAttributes!(
   MaterialAttributes
   Base Color=@n10.RGB
   Specular=@n11.R
   Roughness=@n11.R
   Normal=MaterialFunctionCall!(Normals=@n12.RGB, TransformToTangentSpace=StaticBool() {value=true}, Normal Strength) {materialFunction=/Game/Athena/Apollo/Environments/Landscape/Materials/Biome/CommonMaterials/MF_HQ_WorldAlignedNormal}.Result
   Opacity Mask=@n10.RGB
  ) {attributeSetTypes [
    #2
    #3
    #4
    #1
    #5
   ]}
  n14 = StaticSwitchParameter(True=@n13, False=StaticSwitchParameter(True=@n7, False=@n5) {parameterName=UseGravelAsLayer1}) {group=CustomLayer, parameterName=UseCustomLayerAsLayer1, sortPriority=-9000}
  NamedRerouteDeclaration(None=GetMaterialAttributes(None=NamedRerouteDeclaration(None=@n14) {name=TerrainLayer1, nodeColor=<rgba 1 0 0.7734 1>}) {attributeGetTypes [#5]}.Opacity Mask) {name=TerrainLayer1_Height, nodeColor=<rgba 1 0 0.7734 1>}
  n15 = Constant3Vector() {constant=<rgba 0 0 1 0>}
  n16 = ScalarParameter() {defaultValue=7.6, group=Top Layer, parameterName=Top Layer Blend Sharpness}
  n17 = Multiply(A=NamedRerouteUsage(), B=ScalarParameter() {defaultValue=1, group=Top Layer, parameterName=Top Layer Blend Bias})
  n18 = Multiply(A=NamedRerouteUsage(), B=Constant() {r=-7})
  n19 = Multiply(A=Add(A=NamedRerouteUsage(), B=NamedRerouteUsage()), B=Add(A=Add(A=@n17, B=@n18), B=NamedRerouteUsage()))
  n20 = StaticBool() {value=true}
  NamedRerouteDeclaration(None=Saturate(None=MaterialFunctionCall!(In Explicit Normal=@n15, In World Vector, Blend Sharpness=@n16, Blend Bias=@n19, Clamped?=@n20, Alpha) {materialFunction=/Engine/Functions/Engine_MaterialFunctions01/AlphaBlend/WorldAlignedBlend}.w/Vertex Normals)) {name=TopLayerBlendMask, nodeColor=<rgba 1 0 0.7734 1>}
  n21 = MaterialFunctionCall!(Base Material=NamedRerouteUsage(), Top Material=NamedRerouteUsage(), Alpha=NamedRerouteUsage()) {materialFunction=/Engine/Functions/MaterialLayerFunctions/MatLayerBlend_Standard}
  n22 = Multiply(A=ScalarParameter() {defaultValue=1, group=Detail Normal, parameterName=Detail Normal Intensity}, B=NamedRerouteUsage())
  n23 = TextureSampleParameter2D(UVs=Multiply(A=TextureCoordinate(), B=ScalarParameter() {defaultValue=2, group=Detail Normal, parameterName=Detail Normal Tiling})) {
   group=Detail Normal
   parameterName=Detail Normal
   samplerType=SAMPLERTYPE_Normal
   texture=/Game/Environments/Apollo/Rocks/Natural_Rocks/Textures/T_RockDetails_02_N
  }
  n24 = ShadingPathSwitch(
   Default=QualitySwitch(
    Default=@n21.Blended Material
    Epic=MaterialFunctionCall!(NormalMask=@n22, Material=@n21.Blended Material, Normal=@n23.RGB) {materialFunction=/Engine/Functions/MaterialLayerFunctions/MatLayerBlend_NormalBlend}.Blended Material
   )
   Mobile=@n21.Blended Material
  )
  NamedRerouteDeclaration(None=StaticSwitchParameter(
    True=StaticSwitchParameter(True=@n24, False=@n21.Blended Material) {defaultValue=true, parameterName=Use Detail Normals?}
    False=NamedRerouteUsage()
   ) {defaultValue=true, parameterName=Use Baked Textures}) {name=BakedLayerAndMiddleLayer, nodeColor=<rgba 1 0 0.7734 1>}
  n25 = GetMaterialAttributes(None=@n3.ScreeLayer) {attributeGetTypes [#1]}
  NamedRerouteDeclaration(None=SetMaterialAttributes!(
    MaterialAttributes=@n25.MaterialAttributes
    Normal=Transform(None=@n25.Normal) {transformSourceType=TRANSFORMSOURCE_World, transformType=TRANSFORM_Tangent}
   ) {attributeSetTypes [#1]}) {name=Hermes_ScreeLayer, nodeColor=<rgba 0.9219 0 1 1>}
  n26 = TextureSample(UVs=Multiply(A=ScalarParameter() {defaultValue=1, group=Break Up, parameterName=Breakup Texture Tiling}, B=TextureCoordinate())) {samplerSource=SSM_Wrap_WorldGroupSettings, texture=/Game/Effects/Fort_Effects/Textures/Fire/T_FireNoise_02}
  n27 = QualitySwitch(
   Default=Constant() {r=1}
   Epic=LinearInterpolate(
    A=ScalarParameter() {defaultValue=0.8032, group=Break Up, parameterName=Break Up Bias}
    B=Constant() {r=1}
    Alpha=NamedRerouteDeclaration(None=@n26.R) {name=BreakupNoise, nodeColor=<rgba 0.25 0.25 0.25 1>}
   )
  )
  n28 = VertexColor()
  NamedRerouteDeclaration(None=Multiply(A=@n27, B=@n28.R)) {name=VertMaskRed, nodeColor=<rgba 0.25 0.25 0.25 1>}
  NamedRerouteDeclaration(None=Multiply(A=@n27, B=@n28.B)) {name=VertMaskBlue, nodeColor=<rgba 0.25 0.25 0.25 1>}
  NamedRerouteDeclaration(None=Multiply(A=@n27, B=OneMinus(None=@n28.A))) {name=VertMaskAlpha, nodeColor=<rgba 0.25 0.25 0.25 1>}
  NamedRerouteDeclaration(None=@n28.G) {name=VertMaskGreen, nodeColor=<rgba 0.25 0.25 0.25 1>}
  NamedRerouteDeclaration(None=@n27) {name=BreakupNoiseModified, nodeColor=<rgba 0.25 0.25 0.25 1>}
  n29 = TextureSampleParameter2D() {
   group=Baked
   parameterName=Baked Diffuse
   samplerSource=SSM_Wrap_WorldGroupSettings
   texture=/Game/Environments/Apollo/Sets/Abducted/Chunks/Textures/Mothership/T_Abducted_Chunk_2_Base_D
  }
  n30 = TextureSampleParameter2D() {
   group=Baked
   parameterName=Baked Specular
   samplerSource=SSM_Wrap_WorldGroupSettings
   samplerType=SAMPLERTYPE_Masks
   texture=/Game/Environments/Apollo/Sets/Abducted/Chunks/Textures/Mothership/T_Abducted_Chunk_2_Base_S
  }
  n31 = TextureSampleParameter2D() {
   group=Baked
   parameterName=Baked Normal
   samplerSource=SSM_Wrap_WorldGroupSettings
   samplerType=SAMPLERTYPE_Normal
   texture=/Game/Environments/Apollo/Sets/Abducted/Chunks/Textures/Mothership/T_Abducted_Chunk_2_Base_N
  }
  n32 = SetMaterialAttributes!(
   MaterialAttributes
   Base Color=Multiply(A=@n29.RGB, B=VectorParameter() {group=Baked, parameterName=Baked Modulate color, defaultValue=<rgba 0.6354 0.5082 0.4071 1>}.RGB)
   Specular=NamedRerouteDeclaration(None=@n30.R) {name=BakedBaseSpec, nodeColor=<rgba 1 0 0.7734 1>}
   Roughness=@n30.B
   Normal=@n31.RGB
  ) {attributeSetTypes [
    #2
    #3
    #4
    #1
   ]}
  NamedRerouteDeclaration(None=@n32) {name=BakedBaseLayer, nodeColor=<rgba 1 0 0.7734 1>}
  n33 = StaticSwitchParameter(
   True=MaterialFunctionCall(Base Material=NamedRerouteUsage(), Top Material=NamedRerouteUsage(), Alpha=NamedRerouteUsage()) {materialFunction=&MatLayerBlend_Standard}.Blended Material
   False=NamedRerouteUsage()
  ) {defaultValue=true, parameterName=UseTerrainLayer1}
  n34 = MaterialFunctionCall!(DetailTexture=NamedRerouteUsage(), Mask=NamedRerouteUsage(), NoiseTexture=NamedRerouteUsage(), In=@n33) {materialFunction=&/MF_DemonIslandGlow}
  n35 = NamedRerouteUsage()
  n36 = Multiply(
   A=LinearInterpolate(
    A=ScalarParameter() {defaultValue=2, group=Additional Layer, parameterName=Additional Heightmap Bias B}
    B=ScalarParameter() {defaultValue=0.5, group=Additional Layer, parameterName=Additional Heightmap Bias A}
    Alpha=NamedRerouteUsage()
   )
   B=NamedRerouteUsage()
  )
  n37 = ScalarParameter() {defaultValue=1, group=Additional Layer, parameterName=Additional Heightmap Contrast}
  n38 = StaticSwitchParameter(
   True=MaterialFunctionCall(
    Base Material=@n33
    Top Material=GetMaterialAttributes(None=@n35) {attributeGetTypes [#5]}.MaterialAttributes
    Alpha=MaterialFunctionCall(In=@n36, Contrast=@n37) {materialFunction=&CheapContrast}.Result
   ) {materialFunction=&MatLayerBlend_Standard}.Blended Material
   False=@n34.Result
  ) {parameterName=UseTerrainLayer2}
  StaticSwitchParameter(
   True=@n34.Result
   False=ShadingPathSwitch(Default=QualitySwitch(Default=@n34.Result, Low=@n34.Result, High=@n38, Medium=@n38, Epic=@n38), Mobile=@n34.Result)
  ) {defaultValue=true, parameterName=Use Baked Textures}
 ]
}
